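# Physics asset for the large male Tough Henchman backpack cosmetic, giving the pack and its axe and hammer simple physics bodies.
# A wide capsule covers the pack bone, and thin capsules cover the dyn_axe and dyn_hammer bones.
# Each tool is constrained to the pack with narrow swing and twist limits, the hammer's twist being tighter, which restricts them to slight dangling motion.
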PhysicsAsset /BRCosmetics/Accessories/FORT_Backpacks/Backpack_M_LRG_Henchmen_Tough/Meshes/M_LRG_Henchman_Tough_Physics
bodies {
 dyn_axe [{shapeType=Capsule, length=15, radius=2, center=<2 1 0>, rotation=<rot 90 0 0>}]
 dyn_hammer [{shapeType=Capsule, length=15, radius=2, center=<2 -1.2 0>, rotation=<rot 90 0 0>}]
 pack [{shapeType=Capsule, length=10, radius=13.3, center=<0.0003 -10.8449 -1.9339>}]
}
constraints [
 {bone1Name=dyn_axe, bone2Name=pack, swing1LimitDegrees=20, swing2LimitDegrees=20, twistLimitDegrees=20}
 {bone1Name=dyn_hammer, bone2Name=pack, swing1LimitDegrees=20, swing2LimitDegrees=20, twistLimitDegrees=10}
]
